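# Sound cue for the looping cloud rumble ambience of the Major Spire.
# A single looping wave player feeds the output at reduced cue volume, with a shared cloud attenuation asset shaping it spatially.
# It routes through the Athena outdoor weather ambient sound class and references the spire's ambient concurrency set, alongside override values specifying one instance and prevent-new resolution.
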SoundCue /SpireSharedAssets/Audio/Cues/SC_MajorSpire_Cloud_Rumble_Loop
root=WavePlayer_2
nodes [{id=WavePlayer_2, properties {bLooping=true, soundWaveAssetPtr=/SpireSharedAssets/Audio/Waves/MajorSpire_Cloud_Rumble_Loop}}]
properties {
 attenuationSettings=/SpireSharedAssets/Audio/Settings/ATT_MajorSpire_Cloud
 maxDistance=28000
 soundClassObject=/Game/Sounds/FortSoundClassesAndModes/Fort_Ambient_Outdoors_Weather_Athena
 volumeMultiplier=0.5
 concurrencyOverrides {maxCount=1, resolutionRule=PreventNew, retriggerTime=1}
 concurrencySet [/SpireSharedAssets/Audio/Settings/Conc_MajorSpire_Ambient]
}
