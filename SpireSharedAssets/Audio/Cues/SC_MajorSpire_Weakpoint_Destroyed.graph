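# Sound cue for the destruction of a Major Spire weakpoint, routed to the boss-fight player-critical sound class.
# A slight pitch modulator wraps a quality switch: the first two levels pick randomly from eight wave variants, the last two from a four-wave subset.
# It uses the Spire attack attenuation and priority concurrency assets.
# Bus sends feed sidechain buses for spire loops, weapons, music, ambience and footsteps, presumably to duck them.
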SoundCue /SpireSharedAssets/Audio/Cues/SC_MajorSpire_Weakpoint_Destroyed
root=Modulator_0
nodes [
 {id=Modulator_0, children [QualityLevel_0], properties {pitchMax=1, pitchMin=0.98}}
 {id=QualityLevel_0, children [Random_1,Random_1,Random_2,Random_2]}
 {
  id=Random_1
  children [WavePlayer_3, WavePlayer_4, WavePlayer_5, WavePlayer_6, WavePlayer_7, WavePlayer_9, WavePlayer_10, WavePlayer_11]
  properties {bShouldExcludeFromBranchCulling=true, weights [1,1,1,1,1,1,1,1]}
 }
 {id=Random_2, children [WavePlayer_12, WavePlayer_13, WavePlayer_23, WavePlayer_14], properties {weights [1,1,1,1]}}
 {id=WavePlayer_10, properties {soundWaveAssetPtr=/SpireSharedAssets/Audio/Waves/MajorSpire_Weakpoint_Destroyed_02}}
 {id=WavePlayer_11, properties {soundWaveAssetPtr=/SpireSharedAssets/Audio/Waves/MajorSpire_Weakpoint_Destroyed_01}}
 {id=WavePlayer_12, properties {soundWaveAssetPtr=/SpireSharedAssets/Audio/Waves/MajorSpire_Weakpoint_Destroyed_08}}
 {id=WavePlayer_13, properties {soundWaveAssetPtr=/SpireSharedAssets/Audio/Waves/MajorSpire_Weakpoint_Destroyed_07}}
 {id=WavePlayer_14, properties {soundWaveAssetPtr=&MajorSpire_Weakpoint_Destroyed_01}}
 {id=WavePlayer_23, properties {soundWaveAssetPtr=&MajorSpire_Weakpoint_Destroyed_02}}
 {id=WavePlayer_3, properties {soundWaveAssetPtr=&MajorSpire_Weakpoint_Destroyed_08}}
 {id=WavePlayer_4, properties {soundWaveAssetPtr=&MajorSpire_Weakpoint_Destroyed_07}}
 {id=WavePlayer_5, properties {soundWaveAssetPtr=/SpireSharedAssets/Audio/Waves/MajorSpire_Weakpoint_Destroyed_06}}
 {id=WavePlayer_6, properties {soundWaveAssetPtr=/SpireSharedAssets/Audio/Waves/MajorSpire_Weakpoint_Destroyed_05}}
 {id=WavePlayer_7, properties {soundWaveAssetPtr=/SpireSharedAssets/Audio/Waves/MajorSpire_Weakpoint_Destroyed_04}}
 {id=WavePlayer_9, properties {soundWaveAssetPtr=/SpireSharedAssets/Audio/Waves/MajorSpire_Weakpoint_Destroyed_03}}
]
properties {
 attenuationSettings=/SpireSharedAssets/Audio/Settings/ATT_MajorSpire_Attacks
 maxDistance=16000
 priority=8
 soundClassObject=/Game/Sounds/FortSoundClassesAndModes/Fort_BossFight_SoundClass/Fort_BossFight_PlayerCritical
 volumeMultiplier=2.5
 attenuationOverrides {bAttenuate=false, falloffDistance=10000, attenuationShapeExtents {x=0}}
 busSends [
  {audioBus=/SpireSharedAssets/Audio/Modulation/AB_Spires_Loops_SideChain_Vol, minSendLevel=0, sendLevel=0.8, soundSourceBus=null}
  {audioBus=&AB_Sidechain_FortWeapon_Vol, minSendLevel=0, sendLevel=0.5, soundSourceBus=null}
  {audioBus=&AB_Sidechain_FortMusic_Vol, minSendLevel=0, sendLevel=0.4, soundSourceBus=null}
  {audioBus=&AB_Sidechain_FortAmbient_Vol, minSendLevel=0, sendLevel=0.8, soundSourceBus=null}
  {audioBus=/Game/Sounds/Modulation/EnvelopeFollowers/Volume/AB_Sidechain_Footsteps_Vol, minSendLevel=0, sendLevel=0.8, soundSourceBus=null}
 ]
 concurrencyOverrides {maxCount=2, resolutionRule=StopOldest, retriggerTime=0.5, voiceStealReleaseTime=0.1}
 concurrencySet [/SpireSharedAssets/Audio/Settings/Conc_MajorSpire_Attacks_Pri]
}
