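# Sound cue for the Minor Spire beam attack charge-up, classed as a boss-fight projectile sound with Minor Spire attack attenuation.
# A Random node picks one of five equally weighted charge waves and feeds it through a Modulator.
# Concurrency overrides allow one instance per owner and block new ones rather than cutting the playing one.
# Bus sends feed sidechain buses for spire loops, weapons, music, ambience and footsteps.
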SoundCue /SpireSharedAssets/Audio/Cues/SC_MinorSpire_Beam_ChargeUp
root=Modulator_0
nodes [
 {id=Modulator_0, children [Random_0]}
 {
  id=Random_0
  children [WavePlayer_1,WavePlayer_2,WavePlayer_3,WavePlayer_4,WavePlayer_5]
  properties {bShouldExcludeFromBranchCulling=true, weights [1,1,1,1,1]}
 }
 {id=WavePlayer_1, properties {soundWaveAssetPtr=/SpireSharedAssets/Audio/Waves/MinorSpire_Beam_Charge_05}}
 {id=WavePlayer_2, properties {soundWaveAssetPtr=/SpireSharedAssets/Audio/Waves/MinorSpire_Beam_Charge_04}}
 {id=WavePlayer_3, properties {soundWaveAssetPtr=/SpireSharedAssets/Audio/Waves/MinorSpire_Beam_Charge_03}}
 {id=WavePlayer_4, properties {soundWaveAssetPtr=/SpireSharedAssets/Audio/Waves/MinorSpire_Beam_Charge_02}}
 {id=WavePlayer_5, properties {soundWaveAssetPtr=/SpireSharedAssets/Audio/Waves/MinorSpire_Beam_Charge_01}}
]
properties {
 attenuationSettings=/SpireSharedAssets/Audio/Settings/ATT_MinorSpire_Attacks
 bOverrideConcurrency=true
 maxDistance=16000
 soundClassObject=/Game/Sounds/FortSoundClassesAndModes/Fort_BossFight_SoundClass/Fort_BossFight_Projectile
 volumeMultiplier=2.2
 busSends [
  {audioBus=/SpireSharedAssets/Audio/Modulation/AB_Spires_Loops_SideChain_Vol, minSendLevel=0, sendLevel=0.6, soundSourceBus=null}
  {audioBus=&AB_Sidechain_FortWeapon_Vol, minSendLevel=0, sendLevel=0.3, soundSourceBus=null}
  {audioBus=&AB_Sidechain_FortMusic_Vol, minSendLevel=0, sendLevel=0.2, soundSourceBus=null}
  {audioBus=&AB_Sidechain_FortAmbient_Vol, minSendLevel=0, sendLevel=0.5, soundSourceBus=null}
  {audioBus=/Game/Sounds/Modulation/EnvelopeFollowers/Volume/AB_Sidechain_Footsteps_Vol, minSendLevel=0, sendLevel=0.3, soundSourceBus=null}
 ]
 concurrencyOverrides {bLimitToOwner=true, maxCount=1, resolutionRule=PreventNew, retriggerTime=1}
}
