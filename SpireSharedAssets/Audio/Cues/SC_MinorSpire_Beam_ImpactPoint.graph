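# Sound cue for the impact point of the Minor Spire's beam attack.
# A Random node picks one of four equally weighted impact waves, excluded from branch culling, and a Modulator shifts its pitch slightly upward.
# It uses the Minor Spire attack attenuation and the boss-fight player-critical sound class, with boosted volume.
# A concurrency override allows one instance at a time, preventing new plays, with a retrigger time.
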SoundCue /SpireSharedAssets/Audio/Cues/SC_MinorSpire_Beam_ImpactPoint
root=Modulator_0
nodes [
 {id=Modulator_0, children [Random_2], properties {pitchMax=1.2, pitchMin=1.1}}
 {
  id=Random_2
  children [WavePlayer_0, WavePlayer_11, WavePlayer_12, WavePlayer_13]
  properties {bShouldExcludeFromBranchCulling=true, weights [1,1,1,1]}
 }
 {id=WavePlayer_0, properties {soundWaveAssetPtr=/SpireSharedAssets/Audio/Waves/MinorSpire_Beam_ImpactPoint_01}}
 {id=WavePlayer_11, properties {soundWaveAssetPtr=/SpireSharedAssets/Audio/Waves/MinorSpire_Beam_ImpactPoint_02}}
 {id=WavePlayer_12, properties {soundWaveAssetPtr=/SpireSharedAssets/Audio/Waves/MinorSpire_Beam_ImpactPoint_03}}
 {id=WavePlayer_13, properties {soundWaveAssetPtr=/SpireSharedAssets/Audio/Waves/MinorSpire_Beam_ImpactPoint_04}}
]
properties {
 attenuationSettings=/SpireSharedAssets/Audio/Settings/ATT_MinorSpire_Attacks
 bOverrideConcurrency=true
 maxDistance=16000
 soundClassObject=/Game/Sounds/FortSoundClassesAndModes/Fort_BossFight_SoundClass/Fort_BossFight_PlayerCritical
 volumeMultiplier=2
 concurrencyOverrides {maxCount=1, resolutionRule=PreventNew, retriggerTime=1}
}
