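# Sound cue for hits on a Minor Spire, played through the instigator hit-notification sound class.
# It mixes random impact waves under hit-marker attenuation with quieter hit-react waves under spire attack attenuation, each volume-modulated.
# A quality switch gives the first two levels full variation sets and the last two three-wave subsets of each layer.
# A concurrency override allows three voices per owner, stopping the oldest.
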SoundCue /SpireSharedAssets/Audio/Cues/SC_MinorSpire_HitImpact
root=QualityLevel_0
nodes [
 {id=Attenuation_0, children [Modulator_0], properties {attenuationSettings=/Game/Athena/Sounds/HitNotifications/Fort_HitMarker_Instigate_Att}}
 {id=Attenuation_1, children [Modulator_1], properties {attenuationSettings=/SpireSharedAssets/Audio/Settings/ATT_MinorSpire_Attacks}}
 {id=Attenuation_2, children [Modulator_2], properties {attenuationSettings=&Fort_HitMarker_Instigate_Att}}
 {id=Attenuation_3, children [Modulator_3], properties {attenuationSettings=&ATT_MinorSpire_Attacks}}
 {id=Mixer_0, children [Attenuation_0, Attenuation_1], properties {inputVolume [1,0.3]}}
 {id=Mixer_1, children [Attenuation_2, Attenuation_3], properties {inputVolume [1,0.3]}}
 {id=Modulator_0, children [Random_1], properties {volumeMin=0.5}}
 {id=Modulator_1, children [Random_5], properties {volumeMin=0.3}}
 {id=Modulator_2, children [Random_6], properties {volumeMin=0.5}}
 {id=Modulator_3, children [Random_4], properties {volumeMin=0.3}}
 {id=QualityLevel_0, children [Mixer_0,Mixer_0,Mixer_1,Mixer_1]}
 {
  id=Random_1
  children [WavePlayer_0, WavePlayer_6, WavePlayer_8, WavePlayer_15, WavePlayer_19, WavePlayer_20]
  properties {bShouldExcludeFromBranchCulling=true, weights [1,1,1,1,1,1]}
 }
 {id=Random_4, children [WavePlayer_16, WavePlayer_17, WavePlayer_18], properties {weights [1,1,1]}}
 {
  id=Random_5
  children [WavePlayer_1, WavePlayer_2, WavePlayer_3, WavePlayer_4, WavePlayer_5, WavePlayer_9, WavePlayer_10, WavePlayer_11]
  properties {bShouldExcludeFromBranchCulling=true, weights [1,1,1,1,1,1,1,1]}
 }
 {id=Random_6, children [WavePlayer_7, WavePlayer_21, WavePlayer_22], properties {weights [1,1,1]}}
 {id=WavePlayer_0, properties {soundWaveAssetPtr=/SpireSharedAssets/Audio/Waves/MinorSpire_Impact_01}}
 {id=WavePlayer_1, properties {soundWaveAssetPtr=/SpireSharedAssets/Audio/Waves/MinorSpire_HitReact_01}}
 {id=WavePlayer_10, properties {soundWaveAssetPtr=/SpireSharedAssets/Audio/Waves/MinorSpire_HitReact_07}}
 {id=WavePlayer_11, properties {soundWaveAssetPtr=/SpireSharedAssets/Audio/Waves/MinorSpire_HitReact_08}}
 {id=WavePlayer_15, properties {soundWaveAssetPtr=/SpireSharedAssets/Audio/Waves/MinorSpire_Impact_04}}
 {id=WavePlayer_16, properties {soundWaveAssetPtr=&MinorSpire_HitReact_01}}
 {id=WavePlayer_17, properties {soundWaveAssetPtr=/SpireSharedAssets/Audio/Waves/MinorSpire_HitReact_02}}
 {id=WavePlayer_18, properties {soundWaveAssetPtr=/SpireSharedAssets/Audio/Waves/MinorSpire_HitReact_03}}
 {id=WavePlayer_19, properties {soundWaveAssetPtr=/SpireSharedAssets/Audio/Waves/MinorSpire_Impact_05}}
 {id=WavePlayer_2, properties {soundWaveAssetPtr=&MinorSpire_HitReact_02}}
 {id=WavePlayer_20, properties {soundWaveAssetPtr=/SpireSharedAssets/Audio/Waves/MinorSpire_Impact_06}}
 {id=WavePlayer_21, properties {soundWaveAssetPtr=/SpireSharedAssets/Audio/Waves/MinorSpire_Impact_02}}
 {id=WavePlayer_22, properties {soundWaveAssetPtr=/SpireSharedAssets/Audio/Waves/MinorSpire_Impact_03}}
 {id=WavePlayer_3, properties {soundWaveAssetPtr=&MinorSpire_HitReact_03}}
 {id=WavePlayer_4, properties {soundWaveAssetPtr=/SpireSharedAssets/Audio/Waves/MinorSpire_HitReact_04}}
 {id=WavePlayer_5, properties {soundWaveAssetPtr=/SpireSharedAssets/Audio/Waves/MinorSpire_HitReact_05}}
 {id=WavePlayer_6, properties {soundWaveAssetPtr=&MinorSpire_Impact_02}}
 {id=WavePlayer_7, properties {soundWaveAssetPtr=&MinorSpire_Impact_01}}
 {id=WavePlayer_8, properties {soundWaveAssetPtr=&MinorSpire_Impact_03}}
 {id=WavePlayer_9, properties {soundWaveAssetPtr=/SpireSharedAssets/Audio/Waves/MinorSpire_HitReact_06}}
]
properties {
 bOverrideConcurrency=true
 maxDistance=2097152
 soundClassObject=&Fort_2DHitNotifications_Instigate_PSM
 volumeMultiplier=0.6
 concurrencyOverrides {bLimitToOwner=true, maxCount=3, resolutionRule=StopOldest, voiceStealReleaseTime=0.1, volumeScaleAttackTime=0.1}
}
