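# SC_MinorSpire_Ricochet is a sound cue for minor spire ricochets, built from MajorSpire shield ricochet waves and major spire attenuation.
# A quality-level node picks an equally weighted random pool of twenty waves for its first two levels and a five-wave subset for the other two.
# A modulator above it varies volume with fixed pitch.
# Owner-limited concurrency caps simultaneous plays, stopping the oldest.
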SoundCue /SpireSharedAssets/Audio/Cues/SC_MinorSpire_Ricochet
root=Modulator_0
nodes [
 {id=Modulator_0, children [QualityLevel_0], properties {pitchMax=1, pitchMin=1, volumeMin=0.75}}
 {id=QualityLevel_0, children [Random_0,Random_0,Random_1,Random_1]}
 {
  id=Random_0
  children [
   WavePlayer_0
   WavePlayer_1
   WavePlayer_2
   WavePlayer_3
   WavePlayer_4
   WavePlayer_5
   WavePlayer_6
   WavePlayer_7
   WavePlayer_8
   WavePlayer_9
   WavePlayer_10
   WavePlayer_11
   WavePlayer_12
   WavePlayer_13
   WavePlayer_14
   WavePlayer_15
   WavePlayer_16
   WavePlayer_17
   WavePlayer_18
   WavePlayer_19
  ]
  properties {bShouldExcludeFromBranchCulling=true, weights [1,1,1,1,1,1,1,1,1,1,1,1,1,1,1,1,1,1,1,1]}
 }
 {id=Random_1, children [WavePlayer_21, WavePlayer_20, WavePlayer_24, WavePlayer_23, WavePlayer_22], properties {weights [1,1,1,1,1]}}
 {id=WavePlayer_0, properties {soundWaveAssetPtr=/SpireSharedAssets/Audio/Waves/MajorSpire_Shield_Ricochet_01}}
 {id=WavePlayer_1, properties {soundWaveAssetPtr=/SpireSharedAssets/Audio/Waves/MajorSpire_Shield_Ricochet_02}}
 {id=WavePlayer_10, properties {soundWaveAssetPtr=/SpireSharedAssets/Audio/Waves/MajorSpire_Shield_Ricochet_11}}
 {id=WavePlayer_11, properties {soundWaveAssetPtr=/SpireSharedAssets/Audio/Waves/MajorSpire_Shield_Ricochet_12}}
 {id=WavePlayer_12, properties {soundWaveAssetPtr=/SpireSharedAssets/Audio/Waves/MajorSpire_Shield_Ricochet_13}}
 {id=WavePlayer_13, properties {soundWaveAssetPtr=/SpireSharedAssets/Audio/Waves/MajorSpire_Shield_Ricochet_14}}
 {id=WavePlayer_14, properties {soundWaveAssetPtr=/SpireSharedAssets/Audio/Waves/MajorSpire_Shield_Ricochet_15}}
 {id=WavePlayer_15, properties {soundWaveAssetPtr=/SpireSharedAssets/Audio/Waves/MajorSpire_Shield_Ricochet_16}}
 {id=WavePlayer_16, properties {soundWaveAssetPtr=/SpireSharedAssets/Audio/Waves/MajorSpire_Shield_Ricochet_17}}
 {id=WavePlayer_17, properties {soundWaveAssetPtr=/SpireSharedAssets/Audio/Waves/MajorSpire_Shield_Ricochet_18}}
 {id=WavePlayer_18, properties {soundWaveAssetPtr=/SpireSharedAssets/Audio/Waves/MajorSpire_Shield_Ricochet_19}}
 {id=WavePlayer_19, properties {soundWaveAssetPtr=/SpireSharedAssets/Audio/Waves/MajorSpire_Shield_Ricochet_20}}
 {id=WavePlayer_2, properties {soundWaveAssetPtr=/SpireSharedAssets/Audio/Waves/MajorSpire_Shield_Ricochet_03}}
 {id=WavePlayer_20, properties {soundWaveAssetPtr=/SpireSharedAssets/Audio/Waves/MajorSpire_Shield_Ricochet_05}}
 {id=WavePlayer_21, properties {soundWaveAssetPtr=&MajorSpire_Shield_Ricochet_02}}
 {id=WavePlayer_22, properties {soundWaveAssetPtr=&MajorSpire_Shield_Ricochet_20}}
 {id=WavePlayer_23, properties {soundWaveAssetPtr=&MajorSpire_Shield_Ricochet_16}}
 {id=WavePlayer_24, properties {soundWaveAssetPtr=/SpireSharedAssets/Audio/Waves/MajorSpire_Shield_Ricochet_09}}
 {id=WavePlayer_3, properties {soundWaveAssetPtr=/SpireSharedAssets/Audio/Waves/MajorSpire_Shield_Ricochet_04}}
 {id=WavePlayer_4, properties {soundWaveAssetPtr=&MajorSpire_Shield_Ricochet_05}}
 {id=WavePlayer_5, properties {soundWaveAssetPtr=/SpireSharedAssets/Audio/Waves/MajorSpire_Shield_Ricochet_06}}
 {id=WavePlayer_6, properties {soundWaveAssetPtr=/SpireSharedAssets/Audio/Waves/MajorSpire_Shield_Ricochet_07}}
 {id=WavePlayer_7, properties {soundWaveAssetPtr=/SpireSharedAssets/Audio/Waves/MajorSpire_Shield_Ricochet_08}}
 {id=WavePlayer_8, properties {soundWaveAssetPtr=&MajorSpire_Shield_Ricochet_09}}
 {id=WavePlayer_9, properties {soundWaveAssetPtr=/SpireSharedAssets/Audio/Waves/MajorSpire_Shield_Ricochet_10}}
]
properties {
 attenuationSettings=/SpireSharedAssets/Audio/Settings/ATT_MajorSpire_Ricochets
 bOverrideConcurrency=true
 maxDistance=2097152
 soundClassObject=&Fort_2DHitNotifications_Instigate_PSM
 volumeMultiplier=0.2506
 concurrencyOverrides {bLimitToOwner=true, maxCount=3, resolutionRule=StopOldest, volumeScaleAttackTime=0.1}
 concurrencySet [null]
}
