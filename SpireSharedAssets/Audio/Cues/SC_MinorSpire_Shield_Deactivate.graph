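# Sound cue for the Minor Spire shield deactivation sound, routed to the boss-fight player-critical sound class with shield-specific attenuation.
# A quality-level node feeds two equal-weight random pools: six wave variations for the first two levels, and only the first three of those waves for the last two.
# A concurrency override allows one instance, rejects new plays over that limit and sets a retrigger time.
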SoundCue /SpireSharedAssets/Audio/Cues/SC_MinorSpire_Shield_Deactivate
root=QualityLevel_0
nodes [
 {id=QualityLevel_0, children [Random_0,Random_0,Random_1,Random_1]}
 {id=Random_0, children [WavePlayer_1,WavePlayer_2,WavePlayer_3,WavePlayer_4,WavePlayer_5,WavePlayer_6], properties {weights [1,1,1,1,1,1]}}
 {id=Random_1, children [WavePlayer_7,WavePlayer_8,WavePlayer_9], properties {weights [1,1,1]}}
 {id=WavePlayer_1, properties {soundWaveAssetPtr=/SpireSharedAssets/Audio/Waves/MinorSpire_Shield_Deactivate_01}}
 {id=WavePlayer_2, properties {soundWaveAssetPtr=/SpireSharedAssets/Audio/Waves/MinorSpire_Shield_Deactivate_02}}
 {id=WavePlayer_3, properties {soundWaveAssetPtr=/SpireSharedAssets/Audio/Waves/MinorSpire_Shield_Deactivate_03}}
 {id=WavePlayer_4, properties {soundWaveAssetPtr=/SpireSharedAssets/Audio/Waves/MinorSpire_Shield_Deactivate_04}}
 {id=WavePlayer_5, properties {soundWaveAssetPtr=/SpireSharedAssets/Audio/Waves/MinorSpire_Shield_Deactivate_05}}
 {id=WavePlayer_6, properties {soundWaveAssetPtr=/SpireSharedAssets/Audio/Waves/MinorSpire_Shield_Deactivate_06}}
 {id=WavePlayer_7, properties {soundWaveAssetPtr=&MinorSpire_Shield_Deactivate_01}}
 {id=WavePlayer_8, properties {soundWaveAssetPtr=&MinorSpire_Shield_Deactivate_02}}
 {id=WavePlayer_9, properties {soundWaveAssetPtr=&MinorSpire_Shield_Deactivate_03}}
]
properties {
 attenuationSettings=/SpireSharedAssets/Audio/Settings/ATT_MinorSpire_Shield
 bOverrideConcurrency=true
 maxDistance=10500
 soundClassObject=/Game/Sounds/FortSoundClassesAndModes/Fort_BossFight_SoundClass/Fort_BossFight_PlayerCritical
 volumeMultiplier=1.5036
 concurrencyOverrides {maxCount=1, resolutionRule=PreventNew, retriggerTime=1}
}
