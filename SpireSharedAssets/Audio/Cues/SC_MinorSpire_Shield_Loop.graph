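# Sound cue for the looping shield sound of a Minor Spire, in the boss fight SFX sound class.
# A Random node picks one of three equally weighted looping wave variations.
# Playback uses the Minor Spire ambient attenuation at a reduced volume.
# A concurrency override limits it to a single instance and prevents new ones from starting while it plays.
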SoundCue /SpireSharedAssets/Audio/Cues/SC_MinorSpire_Shield_Loop
root=Random_1
nodes [
 {id=Random_1, children [WavePlayer_0,WavePlayer_1,WavePlayer_2], properties {bShouldExcludeFromBranchCulling=true, weights [1,1,1]}}
 {id=WavePlayer_0, properties {bLooping=true, soundWaveAssetPtr=/SpireSharedAssets/Audio/Waves/MinorSpire_Shield_Loop_01}}
 {id=WavePlayer_1, properties {bLooping=true, soundWaveAssetPtr=/SpireSharedAssets/Audio/Waves/MinorSpire_Shield_Loop_02}}
 {id=WavePlayer_2, properties {bLooping=true, soundWaveAssetPtr=/SpireSharedAssets/Audio/Waves/MinorSpire_Shield_Loop_03}}
]
properties {
 attenuationSettings=/SpireSharedAssets/Audio/Settings/ATT_MinorSpire_Ambient
 bOverrideConcurrency=true
 maxDistance=5500
 soundClassObject=&Fort_BossFight_SFX
 volumeMultiplier=0.1504
 concurrencyOverrides {maxCount=1, resolutionRule=PreventNew, retriggerTime=1}
}
