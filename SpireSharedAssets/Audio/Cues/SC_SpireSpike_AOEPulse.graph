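# SC_SpireSpike_AOEPulse is a Sound Cue for the Spire Spike area-of-effect pulse sound.
# A single Wave Player root plays the SpireSpike_AOEPulse wave directly, with no randomization or modulation nodes.
# The cue uses the shared Fort_Ability_Radius attenuation and the Fort_WeaponMelee_PSM sound class, with its volume boosted above default.
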SoundCue /SpireSharedAssets/Audio/Cues/SC_SpireSpike_AOEPulse
root=WavePlayer_0
nodes [{id=WavePlayer_0, properties {soundWaveAssetPtr=/SpireSharedAssets/Audio/Waves/SpireSpike_AOEPulse}}]
properties {
 attenuationSettings=/Game/Sounds/Attenuation/Fort_Ability_Radius
 maxDistance=1200
 soundClassObject=&Fort_WeaponMelee_PSM
 volumeMultiplier=1.6
}
