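# Looping ambient sound cue for the SpireSpike, routed to the boss fight SFX sound class and spatialized by a SpireSpike attenuation asset.
# A Modulator at default settings feeds from a quality-level switch with four branches.
# The first two branches share an equal-weight Random over six looping waves, excluded from branch culling, while the last two each play one fixed loop.
# An overridden concurrency limit caps simultaneous instances.
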SoundCue /SpireSharedAssets/Audio/Cues/SC_SpireSpike_Ambient_Loop
root=Modulator_1
nodes [
 {id=Modulator_1, children [QualityLevel_1]}
 {id=QualityLevel_1, children [Random_1, Random_1, WavePlayer_9, WavePlayer_10]}
 {
  id=Random_1
  children [WavePlayer_3,WavePlayer_4,WavePlayer_5,WavePlayer_6,WavePlayer_7,WavePlayer_8]
  properties {bShouldExcludeFromBranchCulling=true, weights [1,1,1,1,1,1]}
 }
 {id=WavePlayer_10, properties {bLooping=true, soundWaveAssetPtr=/SpireSharedAssets/Audio/Waves/SpireSpike_Ambient_Loop_02}}
 {id=WavePlayer_3, properties {bLooping=true, soundWaveAssetPtr=/SpireSharedAssets/Audio/Waves/SpireSpike_Ambient_Loop_01}}
 {id=WavePlayer_4, properties {bLooping=true, soundWaveAssetPtr=&SpireSpike_Ambient_Loop_02}}
 {id=WavePlayer_5, properties {bLooping=true, soundWaveAssetPtr=/SpireSharedAssets/Audio/Waves/SpireSpike_Ambient_Loop_03}}
 {id=WavePlayer_6, properties {bLooping=true, soundWaveAssetPtr=/SpireSharedAssets/Audio/Waves/SpireSpike_Ambient_Loop_04}}
 {id=WavePlayer_7, properties {bLooping=true, soundWaveAssetPtr=/SpireSharedAssets/Audio/Waves/SpireSpike_Ambient_Loop_05}}
 {id=WavePlayer_8, properties {bLooping=true, soundWaveAssetPtr=/SpireSharedAssets/Audio/Waves/SpireSpike_Ambient_Loop_06}}
 {id=WavePlayer_9, properties {bLooping=true, soundWaveAssetPtr=&SpireSpike_Ambient_Loop_01}}
]
properties {
 attenuationSettings=/SpireSharedAssets/Audio/Settings/ATT_SpireSpike_Ambient
 bOverrideConcurrency=true
 maxDistance=3200
 soundClassObject=&Fort_BossFight_SFX
 volumeMultiplier=1
 attenuationOverrides {bAttenuate=false, falloffDistance=10000, stereoSpread=500, attenuationShapeExtents {x=2000}}
 concurrencyOverrides {maxCount=3, volumeScaleAttackTime=0.2}
}
