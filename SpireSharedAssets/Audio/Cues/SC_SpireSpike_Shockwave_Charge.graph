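# Sound cue for the Spire Spike shockwave charge, routed to the boss-fight player-critical sound class with spike ambient attenuation.
# A modulator feeds a quality-level switch: the first level randomly picks from three charge waves, the other levels share a two-wave random.
# Sends to loop, weapon, music and ambient sidechain buses suggest it ducks those sounds.
# Concurrency override allows one instance per owner and blocks new retriggers while it plays.
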SoundCue /SpireSharedAssets/Audio/Cues/SC_SpireSpike_Shockwave_Charge
root=Modulator_0
nodes [
 {id=Modulator_0, children [QualityLevel_0]}
 {id=QualityLevel_0, children [Random_0,Random_1,Random_1,Random_1]}
 {id=Random_0, children [WavePlayer_1,WavePlayer_2,WavePlayer_3], properties {bShouldExcludeFromBranchCulling=true, weights [1,1,1]}}
 {id=Random_1, children [WavePlayer_8,WavePlayer_9], properties {weights [1,1]}}
 {id=WavePlayer_1, properties {soundWaveAssetPtr=/SpireSharedAssets/Audio/Waves/SpireSpike_Shockwave_Charge_01}}
 {id=WavePlayer_2, properties {soundWaveAssetPtr=/SpireSharedAssets/Audio/Waves/SpireSpike_Shockwave_Charge_02}}
 {id=WavePlayer_3, properties {soundWaveAssetPtr=/SpireSharedAssets/Audio/Waves/SpireSpike_Shockwave_Charge_03}}
 {id=WavePlayer_8, properties {soundWaveAssetPtr=&SpireSpike_Shockwave_Charge_01}}
 {id=WavePlayer_9, properties {soundWaveAssetPtr=&SpireSpike_Shockwave_Charge_02}}
]
properties {
 attenuationSettings=/SpireSharedAssets/Audio/Settings/ATT_SpireSpike_Ambient
 bOverrideConcurrency=true
 maxDistance=3200
 soundClassObject=/Game/Sounds/FortSoundClassesAndModes/Fort_BossFight_SoundClass/Fort_BossFight_PlayerCritical
 volumeMultiplier=3
 busSends [
  {audioBus=/SpireSharedAssets/Audio/Modulation/AB_Spires_Loops_SideChain_Vol, minSendLevel=0, sendLevel=0.8, soundSourceBus=null}
  {audioBus=&AB_Sidechain_FortWeapon_Vol, minSendLevel=0, sendLevel=0.7, soundSourceBus=null}
  {audioBus=&AB_Sidechain_FortMusic_Vol, minSendLevel=0, sendLevel=0.4, soundSourceBus=null}
  {audioBus=&AB_Sidechain_FortAmbient_Vol, minSendLevel=0, sendLevel=0.9, soundSourceBus=null}
 ]
 concurrencyOverrides {bLimitToOwner=true, maxCount=1, resolutionRule=PreventNew, retriggerTime=1}
}
