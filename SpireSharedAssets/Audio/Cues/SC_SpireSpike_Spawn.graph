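# Sound cue for the Spire Spike spawn sound, using 2D hit-notification sound class and hit-marker attenuation settings.
# A Modulator feeds a quality-level switch: the first level randomly picks from three waves, the other three share a two-wave random pool.
# The three-wave random node is excluded from branch culling.
# Bus sends feed sidechain buses for Spire loops, weapon and ambient volume, and music low-pass filtering.
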SoundCue /SpireSharedAssets/Audio/Cues/SC_SpireSpike_Spawn
root=Modulator_1
nodes [
 {id=Modulator_1, children [QualityLevel_0]}
 {id=QualityLevel_0, children [Random_0,Random_1,Random_1,Random_1]}
 {id=Random_0, children [WavePlayer_0,WavePlayer_1,WavePlayer_2], properties {bShouldExcludeFromBranchCulling=true, weights [1,1,1]}}
 {id=Random_1, children [WavePlayer_6,WavePlayer_7], properties {weights [1,1]}}
 {id=WavePlayer_0, properties {soundWaveAssetPtr=/SpireSharedAssets/Audio/Waves/SpireSpike_Spawn_01}}
 {id=WavePlayer_1, properties {soundWaveAssetPtr=/SpireSharedAssets/Audio/Waves/SpireSpike_Spawn_02}}
 {id=WavePlayer_2, properties {soundWaveAssetPtr=/SpireSharedAssets/Audio/Waves/SpireSpike_Spawn_03}}
 {id=WavePlayer_6, properties {soundWaveAssetPtr=&SpireSpike_Spawn_01}}
 {id=WavePlayer_7, properties {soundWaveAssetPtr=&SpireSpike_Spawn_02}}
]
properties {
 attenuationSettings=/Game/Athena/Sounds/HitNotifications/Fort_HitMarker_Instigate_Att
 maxDistance=2097152
 soundClassObject=&Fort_2DHitNotifications_Instigate_PSM
 volumeMultiplier=1
 attenuationOverrides {bAttenuate=false, falloffDistance=10000, stereoSpread=500, attenuationShapeExtents {x=2000}}
 busSends [
  {audioBus=/SpireSharedAssets/Audio/Modulation/AB_Spires_Loops_SideChain_Vol, minSendLevel=0, sendLevel=0.6, soundSourceBus=null}
  {audioBus=&AB_Sidechain_FortWeapon_Vol, minSendLevel=0, sendLevel=0.5, soundSourceBus=null}
  {audioBus=/Game/Sounds/Modulation/EnvelopeFollowers/LPF/AB_SideChain_FortMusic_LPF, minSendLevel=0, sendLevel=0.2, soundSourceBus=null}
  {audioBus=&AB_Sidechain_FortAmbient_Vol, minSendLevel=0, sendLevel=0.9, soundSourceBus=null}
 ]
 concurrencyOverrides {maxCount=5, resolutionRule=PreventNew}
}
