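# Sound cue for the destruction of a SpireSpike weakpoint, classed as an instigator 2D hit notification.
# A Random node picks one of three equally weighted destroyed-sound waves and feeds it through a Modulator at the root.
# The cue boosts overall volume, and its override settings turn off distance attenuation and define a small concurrency cap that prevents new instances.
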SoundCue /SpireSharedAssets/Audio/Cues/SC_SpireSpike_Weakpoint_Destroyed
root=Modulator_1
nodes [
 {id=Modulator_1, children [Random_0]}
 {id=Random_0, children [WavePlayer_0,WavePlayer_1,WavePlayer_2], properties {weights [1,1,1]}}
 {id=WavePlayer_0, properties {soundWaveAssetPtr=/SpireSharedAssets/Audio/Waves/SpireSpike_Weakpoint_Destroyed_01}}
 {id=WavePlayer_1, properties {soundWaveAssetPtr=/SpireSharedAssets/Audio/Waves/SpireSpike_Weakpoint_Destroyed_02}}
 {id=WavePlayer_2, properties {soundWaveAssetPtr=/SpireSharedAssets/Audio/Waves/SpireSpike_Weakpoint_Destroyed_03}}
]
properties {
 maxDistance=2097152
 soundClassObject=&Fort_2DHitNotifications_Instigate_PSM
 volumeMultiplier=1.25
 attenuationOverrides {bAttenuate=false, falloffDistance=10000, stereoSpread=500, attenuationShapeExtents {x=2000}}
 concurrencyOverrides {maxCount=5, resolutionRule=PreventNew}
}
